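; An actor for a short-lived light beam effect in PistolCanary that stays hidden until its animation is played.
; BeginPlay hides it and creates a dynamic spotlight-billboard material on the light cone mesh with a warm orange color.
; Play Animation reveals it and runs a timeline at the given rate, driving brightness scaled by MaxBrightness and cone X/Y scale while keeping Z.
; The actor destroys itself when the timeline finishes.
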(event EventBeginPlay
  (Rendering|SetActorHiddenInGame true)
  (bind _returnvalue (Rendering|Material|CreateDynamicMaterialInstance (Variables|Default|GetS_Billboard_LightCone) 0 "/Game/Effects/Fort_Effects/Mesh/Rays/M_Billboard_SpotLight.M_Billboard_SpotLight" "None" :!class /Script/Engine.PrimitiveComponent))
  (bind _output_get (Variables|Default|SetDynamicMaterial _returnvalue))
  (Rendering|Material|SetVectorParameterValue _output_get "LightConeColor" "(R=1.000000,G=0.412543,B=0.149960,A=1.000000)" :!class /Script/Engine.MaterialInstanceDynamic))

(event Custom|Play%20Animation
  (params (NewRate :type Float%20%28single-precision%29))
  (bind (_returnvalue_location _returnvalue_rotation _returnvalue_scale) (Math|Transform|BreakTransform (Transformation|GetRelativeTransform (Variables|Default|GetS_Billboard_LightCone))))
  (bind (_returnvalue_scale_x _returnvalue_scale_y _returnvalue_scale_z) (Math|Vector|BreakVector _returnvalue_scale))
  (Rendering|SetActorHiddenInGame false)
  (Components|Timeline|SetPlayRate (Variables|BP_PistolCanary_EndLightBeam|GetLightBeamTimeline) NewRate)
  (bind (_direction _scale _brightness) (Timeline|LightBeamTimeline PlayFromStart)
    (:Update
      (Rendering|Material|SetScalarParameterValue (Variables|Default|GetDynamicMaterial) "Brightness" (* _brightness (Variables|Default|GetMaxBrightness)) :!class /Script/Engine.MaterialInstanceDynamic)
      (Transformation|SetRelativeScale3D (Variables|Default|GetS_Billboard_LightCone) (Math|Vector|MakeVector _scale _scale _returnvalue_scale_z)))
    (:Finished
      (Actor|DestroyActor))))

(fn ConstructionScript)
